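# Parent cloth-capable material for the SnapFreeze Graze cape, adding an elimination-reactive glow between the shared 2023 character PreFX and PostFX functions.
# Time since Elim_StartTime drives a ramp-and-fade envelope, with a secondary mask varying delay, duration and look.
# Gradient and triangle masks blend in color, roughness and noise-panned emissive changes, plus a pulsing Fresnel glow.
# SF_ReactivityToggle blends the effect in, and a debug static switch loops it.
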Material /BRCosmetics/Accessories/FORT_Capes/Cape_F_MED_SnapFreeze/Materials/M_SnapFreeze_Graze_Cape_Parent
graph {
 properties {
  bHasPixelAnimation=true
  bUseMaterialAttributes=true
  bUsedWithClothing=true
  bUsedWithMeshDeformer=true
  bUsedWithSkeletalMesh=true
  pixelDepthOffsetMode=PDOM_Legacy
  shadingModel=MSM_FromMaterialExpression
  usedShadingModels=Default Lit | Subsurface | Clear Coat | Subsurface Profile | Cloth
 }
 nodes [
  n1 = NamedRerouteUsage()
  n2 = GetMaterialAttributes(None=NamedRerouteUsage()) {attributeGetTypes [#1, #2, #3]}
  n3 = Add(A=@n2.Emissive Color, B=NamedRerouteUsage())
  n4 = Subtract(
   A=@n2.Roughness
   B=LinearInterpolate(
    A=ScalarParameter() {defaultValue=0.15, parameterName=Elim_RoughnessBias}
    B=ScalarParameter() {defaultValue=0.3, parameterName=Elim_RoughnessBias_Secondary}
    Alpha=NamedRerouteUsage()
   )
  ) {constB=0.3}
  n5 = LinearInterpolate(
   A=@n2.Base Color
   B=LinearInterpolate(
    A=VectorParameter() {parameterName=Elim_BaseColorHigh, defaultValue=<rgba 0.4229 0.3296 0.4688 1>}.RGB
    B=VectorParameter() {parameterName=Elim_BaseColorLow, defaultValue=<rgba 0.0111 0.0017 0.1146 1>}.RGB
    Alpha=NamedRerouteUsage()
   )
   Alpha=NamedRerouteUsage()
  )
  n6 = GetMaterialAttributes(None=BlendMaterialAttributes(
    A=NamedRerouteUsage()
    B=SetMaterialAttributes!(MaterialAttributes=@n1, Emissive Color=@n3, Base Color=@n5, Roughness=@n4) {attributeSetTypes [#1, #2, #3]}
    Alpha=NamedRerouteUsage()
   )) {attributeGetTypes [#1]}
  n7 = Add(A=@n6.Emissive Color, B=NamedRerouteUsage())
  n8 = MaterialFunctionCall() {materialFunction=&MF_CharacterParent_2023_PreFX}
  MaterialFunctionCall(
   Material=BlendMaterialAttributes(
    A=NamedRerouteUsage()
    B=SetMaterialAttributes!(MaterialAttributes=@n6.MaterialAttributes, Emissive Color=@n7) {attributeSetTypes [#1]}
    Alpha=ScalarParameter() {defaultValue=1, parameterName=SF_ReactivityToggle}
   )
   M=@n8.M
   Pre Skinned Local Position=@n8.M
   Distance Blend=@n8.M
   Fresnel=@n8.M
  ) {materialFunction=&MF_CharacterParent_2023_PostFX}
  n9 = ScalarParameter() {defaultValue=0.4, parameterName=Elim_FresnelExp}
  n10 = ScalarParameter() {defaultValue=0.5, parameterName=Elim_FresnelExp_Secondary}
  n11 = NamedRerouteUsage()
  NamedRerouteDeclaration(None=Fresnel(ExponentIn=LinearInterpolate(A=@n9, B=@n10, Alpha=@n11)) {baseReflectFraction=0, exponent=2}) {name=Fresnel, nodeColor=<rgba 0.6641 0 1 1>}
  n12 = TextureSampleParameter2D() {
   parameterName=Elim_Mask
   samplerType=SAMPLERTYPE_LinearColor
   texture=/BRCosmetics/Accessories/FORT_Capes/Cape_F_MED_SnapFreeze/Textures/T_SnapFreeze_Graze_Cape_GM
  }
  NamedRerouteDeclaration(None=@n12.B) {name=SecondaryMask, nodeColor=<rgba 0 0.25 1 1>}
  n13 = TextureSample(UVs=Panner(Coordinate=TextureCoordinate() {uTiling=6, vTiling=0.5}) {speedX=0.02, speedY=0.2}) {samplerSource=SSM_Wrap_WorldGroupSettings, samplerType=SAMPLERTYPE_Grayscale, texture=&LowResBlurredNoise}
  NamedRerouteDeclaration(None=@n13.R) {name=PanningNoise, nodeColor=<rgba 0 0.0625 1 1>}
  n14 = TextureSampleParameter2D() {
   parameterName=Elim_GradientMask
   samplerType=SAMPLERTYPE_Masks
   texture=/BRCosmetics/Accessories/FORT_Capes/Cape_F_MED_SnapFreeze/Textures/T_SnapFreeze_Graze_Cape_FX
  }
  n15 = Power(Base=OneMinus(None=@n14.R))
  n16 = Time()
  n17 = ScalarParameter() {defaultValue=5, parameterName=Elim_Duration}
  n18 = ScalarParameter() {defaultValue=8, parameterName=Elim_Duration_Secondary}
  n19 = Max(A=@n17, B=@n18)
  n20 = StaticSwitchParameter(True=Modulo!(A=@n16, B=Add(A=@n19) {constB=2}), False=@n16) {parameterName=Debug Elim Reactivity}
  n21 = ScalarParameter() {parameterName=Elim_StartTime}
  n22 = Saturate(None=Divide(A=Subtract(A=@n20, B=@n21), B=ScalarParameter() {defaultValue=0.25, parameterName=Elim_RampUpDuration}))
  n23 = Subtract(A=@n20, B=LinearInterpolate(A=@n21, B=Add(A=@n21) {constB=0.15}, Alpha=NamedRerouteUsage()))
  n24 = LinearInterpolate(A=@n17, B=@n18, Alpha=NamedRerouteUsage())
  n25 = OneMinus(None=Saturate(None=Divide(A=@n23, B=@n24)))
  n26 = NamedRerouteDeclaration(None=Power(Base=Multiply(A=@n22, B=@n25), Exp=ScalarParameter() {defaultValue=2, parameterName=Elim_Exp})) {name=ElimProgress, nodeColor=<rgba 0 0.4609 1 1>}
  n27 = Multiply(A=@n15, B=@n26)
  n28 = OneMinus(None=NamedRerouteUsage())
  n29 = Multiply(
   A=LinearInterpolate(
    A=VectorParameter() {parameterName=Elim_ColorLow, defaultValue=<rgba 0.0111 0.0017 0.1146 1>}.RGB
    B=VectorParameter() {parameterName=Elim_ColorHigh, defaultValue=<rgba 0.4181 0.0365 1 1>}.RGB
    Alpha=Multiply(A=@n27, B=@n28)
   )
   B=@n27
  )
  n30 = Multiply(
   A=LinearInterpolate(
    A=ScalarParameter() {defaultValue=15, parameterName=Elim_Brightness}
    B=ScalarParameter() {defaultValue=8, parameterName=Elim_Brightness_Secondary}
    Alpha=NamedRerouteUsage()
   )
   B=@n13.R
  )
  n31 = VectorParameter() {parameterName=Elim_TrianglesColor, defaultValue=<rgba 0.9906 0 2.5 1>}
  n32 = Multiply(A=NamedRerouteUsage(), B=NamedRerouteUsage())
  NamedRerouteDeclaration(None=Add(A=Multiply(A=@n29, B=@n30), B=Multiply(A=@n31.RGB, B=@n32))) {name=Emissive, nodeColor=<rgba 1 0.9609 0 1>}
  n33 = NamedRerouteDeclaration(None=@n12.G) {name=TrianglesMask, nodeColor=<rgba 1 0.8203 0 1>}
  n34 = Multiply(A=@n15, B=Power(Base=OneMinus(None=@n33)))
  NamedRerouteDeclaration(None=@n34) {name=BaseColorGradient, nodeColor=<rgba 0.6406 0 1 1>}
  NamedRerouteDeclaration(None=Add(A=Multiply(A=@n26, B=@n34), B=@n32)) {name=ReactiveLayerMask, nodeColor=<rgba 0 1 0.3203 1>}
  NamedRerouteDeclaration(None=@n8.Result) {name=InMA, nodeColor=<rgba 0.0078 1 0 1>}
  n35 = Fresnel() {baseReflectFraction=0, exponent=1.5}
  n36 = NamedRerouteUsage()
  n37 = Constant() {r=0.5}
  n38 = Multiply(
   A=Multiply(A=Multiply(A=@n35, B=@n36), B=NamedRerouteUsage())
   B=Clamp(None=MaterialFunctionCall(Speed=@n37) {materialFunction=&DebugTimeSine}.Result) {minDefault=0.5}
  )
  n39 = ScalarParameter() {defaultValue=2, parameterName=Elim_FrenselIntensity}
  NamedRerouteDeclaration(None=Multiply(A=Multiply(A=@n38, B=@n39), B=VectorParameter() {parameterName=Elim_FresnelColor, defaultValue=<rgba 0.3856 0 1 1>}.RGB)) {name=FresnelColor, nodeColor=<rgba 1 0.7969 0 1>}
 ]
}
