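# Control Rig for the SnapFreeze cape cosmetic that corrects shoulder and fur bones as the wearer's arms and head move.
# A first pass plane-clamps both shoulder mover bones and the right shoulder top against arm twist and shoulderpad controls, with debug drawing on.
# Spherical pose readers on both upper arms and the head then store "Back" weights in variables.
# A second pass uses those weights to blend offsets onto the shoulder and back fur bones.
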ControlRigBlueprint /BRCosmetics/Accessories/FORT_Capes/Cape_F_MED_SnapFreeze/Meshes/F_MED_SnapFreeze_Cape_CtrlRig
graphs [
 RigVMFunctionLibrary {links [], nodes []}
 RigVMModel {nodes [
   RigUnit_BeginExecution = &RigUnit_BeginExecution()
   RigVMFunction_Sequence = &RigVMFunction_Sequence(ExecuteContext=@RigUnit_BeginExecution.ExecutePin)
   &RigUnit_SetTransform(
    Item=(Type=Bone,Name="dyn_shoulderMover_r")
    Weight=1
    bInitial=False
    Value=/Script/RigVM.RigVMFunction_MathTransformClampSpatially(
     Axis=Y
     DebugColor=(R=1,G=0,B=0,A=1)
     DebugThickness=1
     Maximum=200
     Minimum=1
     Type=Plane
     bDrawDebug=True
     Space=&RigUnit_GetTransform(Item=(Type=Control,Name="upperarm_twist_01_r_ctrl"), bInitial=False).Transform
     Value=&RigUnit_GetTransform(Item=(Type=Bone,Name="dyn_shoulderMover_r"), bInitial=False).Transform
    ) {
     is=struct RigVMUnitNode
     template="ClampSpatially::Execute(in Value,in Axis,in Type,in Minimum,in Maximum,in Space,in bDrawDebug,in DebugColor,in DebugThickness,out Result)"
    }.Result
    ExecutePin=@RigVMFunction_Sequence.A
   )
   > &RigUnit_SetTransform(
    Item=(Type=Bone,Name="dyn_shoulderMover_l")
    Weight=1
    bInitial=False
    Value=&RigVMFunction_MathTransformClampSpatially(
     Axis=Y
     DebugColor=(R=1,G=0,B=0,A=1)
     DebugThickness=1
     Maximum=200
     Minimum=5
     Type=Plane
     bDrawDebug=True
     Space=&RigUnit_GetTransform(Item=(Type=Control,Name="upperarm_twist_01_l_ctrl"), bInitial=False).Transform
     Value=&RigUnit_GetTransform(Item=(Type=Bone,Name="dyn_shoulderMover_l"), bInitial=False).Transform
    ) {template="ClampSpatially::Execute(in Value,in Axis,in Type,in Minimum,in Maximum,in Space,in bDrawDebug,in DebugColor,in DebugThickness,out Result)"}.Result
   )
   > &RigUnit_SetTransform(
    Item=(Type=Bone,Name="dyn_shoulderTop_r")
    Weight=1
    bInitial=False
    Value=&RigVMFunction_MathTransformClampSpatially(
     Axis=Y
     DebugColor=(R=1,G=0,B=0,A=1)
     DebugThickness=1
     Maximum=200
     Minimum=3
     Type=Plane
     bDrawDebug=True
     Space=&RigUnit_GetTransform(Item=(Type=Control,Name="dyn_shoulderpad_r_ctrl"), bInitial=False).Transform
     Value=&RigUnit_GetTransform(Item=(Type=Bone,Name="dyn_shoulderTop_r"), bInitial=False).Transform
    ) {template="ClampSpatially::Execute(in Value,in Axis,in Type,in Minimum,in Maximum,in Space,in bDrawDebug,in DebugColor,in DebugThickness,out Result)"}.Result
   )
   > n1 = &RigUnit_SphericalPoseReader(
    ActiveRegionScaleFactors=(PositiveWidth=1,NegativeWidth=1,PositiveHeight=1,NegativeHeight=1)
    ActiveRegionSize=0.2
    Debug=(bDrawDebug=True,DebugScale=25,DebugSegments=20,DebugThickness=0.25)
    DriverAxis=(X=0,Y=0,Z=-1)
    DriverItem=(Type=Bone,Name="upperarm_r")
    FalloffRegionScaleFactors=(PositiveWidth=3,NegativeWidth=3,PositiveHeight=3,NegativeHeight=2.3)
    FalloffSize=0.3
    FlipHeightScaling=true
    OptionalParentItem=(Type=Bone)
    RotationOffset=(X=-242.5888,Y=-52.0034,Z=-8.3834)
   )
   > RigVMVariableNode(Value=@n1.OutputParam) {variable=R Upperarm Back}
   > n2 = &RigUnit_SphericalPoseReader(
    ActiveRegionScaleFactors=(PositiveWidth=1,NegativeWidth=1.2,PositiveHeight=1.2,NegativeHeight=1)
    ActiveRegionSize=0.2
    Debug=(bDrawDebug=True,DebugScale=25,DebugSegments=20,DebugThickness=0.25)
    DriverAxis=(X=0,Y=0,Z=1)
    DriverItem=(Type=Bone,Name="upperarm_l")
    FalloffRegionScaleFactors=(PositiveWidth=3,NegativeWidth=3,PositiveHeight=3,NegativeHeight=3)
    FalloffSize=0.3
    FlipHeightScaling=true
    OptionalParentItem=(Type=Bone)
    RotationOffset=(X=-43.3986,Y=-56.0077,Z=-8.3834)
   )
   > RigVMVariableNode(Value=@n2.OutputParam) {variable=L Upperarm Back}
   > n3 = &RigUnit_SphericalPoseReader(
    ActiveRegionScaleFactors=(PositiveWidth=1,NegativeWidth=1,PositiveHeight=1,NegativeHeight=1)
    ActiveRegionSize=0.1
    Debug=(bDrawDebug=true,bDraw2D=False,bDrawLocalAxes=False,DebugScale=25,DebugSegments=20,DebugThickness=0.25)
    DriverAxis=(X=1,Y=0,Z=0)
    DriverItem=(Type=Bone,Name="head")
    FalloffRegionScaleFactors=(PositiveWidth=1.5,NegativeWidth=1.5,PositiveHeight=1.5,NegativeHeight=1.5)
    FalloffSize=0.22
    FlipHeightScaling=False
    FlipWidthScaling=False
    OptionalParentItem=(Type=Bone,Name="None")
    RotationOffset=(X=82.7523,Y=-19.6663,Z=-114.9733)
   )
   > RigVMVariableNode(Value=@n3.OutputParam) {variable=Head Back}
   AlphaInterp = &RigVMFunction_AlphaInterp(
    InterpSpeedDecreasing=30
    InterpSpeedIncreasing=30
    Scale=1
    bClampResult=False
    bInterpResult=False
    Value=RigVMVariableNode() {variable=R Upperarm Back}.Value
   )
   &RigUnit_ModifyTransforms(
    ItemToModify="((Item=(Type=Bone,Name=\"dyn_shoulderMover_r\"),Transform=(Rotation=(X=0,Y=0,Z=0.1736,W=0.9848),Translation=(X=3,Y=-3,Z=0),Scale3D=(X=1,Y=1,Z=1))),(Item=(Type=Bone,Name=\"dyn_shoulderTop_r\"),Transform=(Rotation=(X=0,Y=0,Z=0,W=1),Translation=(X=0,Y=-5,Z=0),Sca ...(truncated)"
    WeightMinimum=0
    Weight=@AlphaInterp.Result
    ExecutePin=@RigVMFunction_Sequence.B
   )
   n4 = &RigVMFunction_AlphaInterp(
    InterpSpeedDecreasing=30
    InterpSpeedIncreasing=30
    Scale=1
    bClampResult=False
    bInterpResult=False
    Value=RigVMVariableNode() {variable=L Upperarm Back}.Value
   )
   > &RigUnit_ModifyTransforms(
    ItemToModify=((Item=(Type=Bone,Name="dyn_shoulderMover_l"),Transform=(Rotation=(X=-0.0241,Y=-0.1738,Z=-0.1351,W=0.9752),Translation=(X=-3,Y=-3,Z=-0.5286),Scale3D=(X=1,Y=1,Z=1))))
    WeightMinimum=0
    Weight=@n4.Result
   )
   > &RigUnit_ModifyTransforms(
    ItemToModify=((Item=(Type=Bone,Name="dyn_shoulderTop_r"),Transform=(Rotation=(X=0,Y=0,Z=0,W=1),Translation=(X=0,Y=0,Z=0),Scale3D=(X=1,Y=1,Z=1))))
    Weight=1
    WeightMinimum=0
   )
   > &RigUnit_ModifyTransforms(
    ItemToModify=((Item=(Type=Bone,Name="dyn_shoulderTop_l"),Transform=(Rotation=(X=0,Y=0,Z=0,W=1),Translation=(X=0,Y=0,Z=0),Scale3D=(X=1,Y=1,Z=1))))
    Weight=1
    WeightMinimum=0
   )
   > &RigUnit_ModifyTransforms(
    ItemToModify=((Item=(Type=Bone,Name="dyn_furBack"),Transform=(Rotation=(X=0,Y=0.4226,Z=0,W=0.9063),Translation=(X=0,Y=0,Z=0),Scale3D=(X=1,Y=1,Z=1))))
    WeightMinimum=0
    Weight=RigVMVariableNode() {variable=Head Back}.Value
   )
  ]}
]
hierarchy [
 root {loc=<0 0 0>, rot=<0 0 0 1>}
 pelvis {parent=root, loc=<0 -0.5301 98.0571>, rot=<0.0127 0.707 -0.0127 -0.707>}
 spine_01 {parent=pelvis, loc=<6.1208 -0.9328 0>, rot=<0 0 0.034 -0.9994>}
 spine_02 {parent=spine_01, loc=<6.5996 -0.2095 0>, rot=<0 0 0.0238 -0.9997>}
 spine_03 {parent=spine_02, loc=<6.4914 -0.1237 0>, rot=<0 0 0.0149 -0.9999>}
 spine_04 {parent=spine_03, loc=<6.9598 0.0794 0>, rot=<0 0 -0.0132 -0.9999>}
 spine_05 {parent=spine_04, loc=<6.9987 0 0>, rot=<0 0 -0.0015 -1>}
 attach_cape {parent=spine_05, loc=<9.3216 -8.4501 0>, rot=<0.0616 -0.7044 -0.0616 -0.7044>}
 clavicle_l {parent=spine_05, loc=<5.4831 -0.1053 -3.2946>, rot=<0.0465 -0.7289 -0.009 -0.6829>}
 upperarm_l {parent=clavicle_l, loc=<9.8658 -1.2136 -0.4105>, rot=<0.0164 -0.3245 -0.0269 -0.9454>}
 upperarm_twist_01_l {parent=upperarm_l, loc=<7.2135 -0.1134 0>, rot=<0 0 0 1>}
 deform_upperarm_bk_base_l {parent=upperarm_twist_01_l, loc=<-6.4053 -4.3746 0.314>, rot=<-0.0065 0.3563 0.0039 -0.9343>}
 deform_upperarm_bk_l {parent=deform_upperarm_bk_base_l, loc=<0 0 0>, rot=<0 0 0 1>}
 deform_upperarm_fr_base_l {parent=upperarm_twist_01_l, loc=<-5.6312 5.092 0.3354>, rot=<-0.0065 0.3563 0.0039 -0.9343>}
 deform_upperarm_fr_l {parent=deform_upperarm_fr_base_l, loc=<0 0 0>, rot=<0 0 0 1>}
 deform_upperarm_up_base_l {parent=upperarm_twist_01_l, loc=<-8.891 0.4825 3.9517>, rot=<-0.0065 0.3563 0.0039 -0.9343>}
 deform_upperarm_up_l {parent=deform_upperarm_up_base_l, loc=<0 0 0>, rot=<0 0 0 1>}
 upperarm_twist_01_l_ctrl {controlType=EULER_TRANSFORM, parent=upperarm_twist_01_l, type=Control, loc=<0 0 0>, rot=<0 0 0 1>}
 upperarm_twist_02_l {parent=upperarm_l, loc=<19.7998 -0.1134 0>, rot=<0 0 0 1>}
 dyn_shoulderpad_l_base {parent=clavicle_l, loc=<9.8658 -1.2136 -0.4105>, rot=<0.0075 0.1927 0.0586 -0.9795>}
 dyn_shoulderpad_l {parent=dyn_shoulderpad_l_base, loc=<6.2726 0 0>, rot=<0 0.2588 0 -0.9659>}
 dyn_shoulderpad_l_ctrl {controlType=EULER_TRANSFORM, parent=dyn_shoulderpad_l, type=Control, loc=<0 0 0>, rot=<0 0 0 1>}
 attach_backpack {parent=spine_05, loc=<0 -8.35 0>, rot=<-0.0093 -0.707 0.0093 -0.707>}
 neck_01 {parent=spine_05, loc=<11.3138 0.1529 0>, rot=<0 0 -0.2003 -0.9797>}
 neck_02 {parent=neck_01, loc=<4.5627 0 0>, rot=<0 0 0.0504 -0.9987>}
 head {parent=neck_02, loc=<5.3914 0 0>, rot=<0 0 0.115 -0.9934>}
 clavicle_r {parent=spine_05, loc=<5.4831 -0.1053 3.2946>, rot=<0.7289 0.0465 -0.6829 0.009>}
 upperarm_r {parent=clavicle_r, loc=<-9.8658 1.2136 0.4105>, rot=<0.0164 -0.3245 -0.0269 -0.9454>}
 upperarm_twist_01_r {parent=upperarm_r, loc=<-7.2135 0.1134 0>, rot=<0 0 0 1>}
 deform_upperarm_bk_base_r {parent=upperarm_twist_01_r, loc=<6.4053 4.3746 -0.314>, rot=<-0.0065 0.3563 0.0039 -0.9343>}
 deform_upperarm_bk_r {parent=deform_upperarm_bk_base_r, loc=<0 0 0>, rot=<0 0 0 1>}
 deform_upperarm_fr_base_r {parent=upperarm_twist_01_r, loc=<5.6312 -5.092 -0.3354>, rot=<-0.0065 0.3563 0.0039 -0.9343>}
 deform_upperarm_fr_r {parent=deform_upperarm_fr_base_r, loc=<0 0 0>, rot=<0 0 0 1>}
 deform_upperarm_up_base_r {parent=upperarm_twist_01_r, loc=<8.891 -0.4825 -3.9517>, rot=<-0.0065 0.3563 0.0039 -0.9343>}
 deform_upperarm_up_r {parent=deform_upperarm_up_base_r, loc=<0 0 0>, rot=<0 0 0 1>}
 upperarm_twist_01_r_ctrl {controlType=EULER_TRANSFORM, parent=upperarm_twist_01_r, type=Control, loc=<0 0 0>, rot=<0 0 0 1>}
 upperarm_twist_02_r {parent=upperarm_r, loc=<-19.7998 0.1134 0>, rot=<0 0 0 1>}
 dyn_shoulderpad_r_base {parent=clavicle_r, loc=<-9.8658 1.2136 0.4105>, rot=<-0.0586 -0.9795 0.0075 -0.1927>}
 dyn_shoulderpad_r {parent=dyn_shoulderpad_r_base, loc=<6.2727 0 0>, rot=<0 0.2588 0 -0.9659>}
 dyn_shoulderpad_r_ctrl {controlType=EULER_TRANSFORM, parent=dyn_shoulderpad_r, type=Control, loc=<0 0 0>, rot=<0 0 0 1>}
 backpack_root {parent=spine_05, loc=<7.2558 -8.5599 0>, rot=<0.0294 -0.7065 -0.0294 -0.7065>}
 hider {parent=backpack_root, loc=<0 -2.5808 0.0076>, rot=<0 0 0 1>}
 cape_top {parent=hider, loc=<0 0.0288 -2.7904>, rot=<0 0 0 1>}
 dyn_cape_r_MSTR {parent=cape_top, loc=<-16.2328 1.5541 -2.5734>, rot=<-0.4358 -0.4358 0.5569 -0.5569>}
 dyn_cape_r_01 {parent=dyn_cape_r_MSTR, loc=<0.3 0 0>, rot=<0 0 0 1>}
 dyn_cape_r_02 {parent=dyn_cape_r_01, loc=<15.8396 0 -1.0508>, rot=<0 -0.0648 0 -0.9979>}
 dyn_cape_r_03 {parent=dyn_cape_r_02, loc=<16.1116 0 0.0718>, rot=<0 -0.0149 0 -0.9999>}
 dyn_cape_r_04 {parent=dyn_cape_r_03, loc=<16.1325 0 0.3367>, rot=<0 -0.0046 0 -1>}
 dyn_cape_r_05 {parent=dyn_cape_r_04, loc=<16.1339 0 0.3952>, rot=<0 0 0 1>}
 dyn_cape_r_06 {parent=dyn_cape_r_05, loc=<16.1245 0 0.2758>, rot=<0 0 0 1>}
 dyn_cape_r_07 {parent=dyn_cape_r_06, loc=<16.1428 0 0.513>, rot=<0 0 0 1>}
 dyn_cape_l_MSTR {parent=cape_top, loc=<16.2328 1.5541 -2.5734>, rot=<-0.4358 -0.4358 0.5569 -0.5569>}
 dyn_cape_l_01 {parent=dyn_cape_l_MSTR, loc=<0.3 0 0>, rot=<0 0 0 1>}
 dyn_cape_l_02 {parent=dyn_cape_l_01, loc=<15.8397 0 -1.0508>, rot=<0 -0.0648 0 -0.9979>}
 dyn_cape_l_03 {parent=dyn_cape_l_02, loc=<16.1116 0 0.0718>, rot=<0 -0.0149 0 -0.9999>}
 dyn_cape_l_04 {parent=dyn_cape_l_03, loc=<16.1324 0 0.3367>, rot=<0 -0.0046 0 -1>}
 dyn_cape_l_05 {parent=dyn_cape_l_04, loc=<16.1338 0 0.3952>, rot=<0 0 0 1>}
 dyn_cape_l_06 {parent=dyn_cape_l_05, loc=<16.1246 0 0.2759>, rot=<0 0 0 1>}
 dyn_cape_l_07 {parent=dyn_cape_l_06, loc=<16.1429 0 0.5129>, rot=<0 0 0 1>}
 dyn_cape_MSTR {parent=cape_top, loc=<0 -0.0317 -2.5734>, rot=<-0.4358 -0.4358 0.5569 -0.5569>}
 dyn_cape_01 {parent=dyn_cape_MSTR, loc=<0.3 0 0>, rot=<0 0 0 1>}
 dyn_cape_02 {parent=dyn_cape_01, loc=<15.8397 0 -1.0508>, rot=<0 -0.0648 0 -0.9979>}
 dyn_cape_03 {parent=dyn_cape_02, loc=<16.1116 0 0.0718>, rot=<0 -0.0149 0 -0.9999>}
 dyn_cape_04 {parent=dyn_cape_03, loc=<16.1324 0 0.3367>, rot=<0 -0.0046 0 -1>}
 dyn_cape_05 {parent=dyn_cape_04, loc=<16.1338 0 0.3952>, rot=<0 0 0 1>}
 dyn_cape_06 {parent=dyn_cape_05, loc=<16.1246 0 0.2759>, rot=<0 0 0 1>}
 dyn_cape_07 {parent=dyn_cape_06, loc=<16.1429 0 0.5129>, rot=<0 0 0 1>}
 dyn_shoulderMover_r {parent=backpack_root, loc=<-13.6014 -1.432 -4.4282>, rot=<-0.0015 0 0 -1>}
 dyn_shoulderMover_l {parent=backpack_root, loc=<13.601 -1.432 -4.4282>, rot=<-0.0015 0 0 -1>}
 dyn_furBack {parent=backpack_root, loc=<0 1.0901 7.2716>, rot=<0.153 0.153 0.6903 -0.6903>}
 dyn_shoulderTop_r {parent=backpack_root, loc=<-10.7409 1.0811 4.2602>, rot=<-0.0015 0 0 -1>}
 dyn_shoulderTop_l {parent=backpack_root, loc=<10.741 1.0811 4.2602>, rot=<-0.0015 0 0 -1>}
]
variables [
 Head Back {type=double, default=0}
 L Upperarm Back {type=double, default=0}
 R Upperarm Back {type=double, default=0}
]
